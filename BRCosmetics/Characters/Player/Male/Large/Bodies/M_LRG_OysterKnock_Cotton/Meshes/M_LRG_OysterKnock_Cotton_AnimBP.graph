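# Post-process style AnimBP for the OysterKnock Cotton large male body cosmetic, adding secondary motion atop the parent mesh's copied pose.
# The head gets an additive counter-rotation, then a Control Rig receives pistol, emote, skydiving, parachute and frontend alphas from the EventGraph.
# Chained spring-driven AnimDynamics nodes simulate knee, inner, outer and belt rope bones.
# EventGraph also computes gravity, sim space and dynamics alpha, with no visible AnimGraph connection.
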AnimBlueprint /BRCosmetics/Characters/Player/Male/Large/Bodies/M_LRG_OysterKnock_Cotton/Meshes/M_LRG_OysterKnock_Cotton_AnimBP
target_skeleton=&/M_LRG_OysterKnock_Cotton_Skeleton
graphs [
 AnimGraph {nodes [
   n1 = "Transform (Modify) Bone - Bone: head"(
    ComponentPose=Local To Component(LocalPose=Copy Pose From Mesh() {bUseAttachedParent=true})
    Alpha=1
    Translation="0,0,0"
    Rotation=Get Head Rot() -> (Head Rot:Rotator)
    Scale="1,1,1"
   ) {rotationMode=BMM_Additive, rotationSpace=BCS_BoneSpace, alphaBoolBlend {blendOption=Linear}, boneToModify {boneName=head}}
   n2 = Control Rig!(
    Source=Component To Local(ComponentPose=@n1)
    Alpha
    IsPistol:double=Get Is Pistol Alpha() -> (Is Pistol Alpha:double)
    Is not playing emote Alpha:double=Get Is not Playing Emote Alpha() -> (Is not Playing Emote Alpha:double)
    Is Skydiving Alpha:double=Get Is Skydiving Alpha() -> (Is Skydiving Alpha:double)
    Is Parachute Opened Alpha:double=Get Is Parachute Opened() -> (Is Parachute Opened:double)
    Is In Frontend Alpha:double=Get Is in FrontEnd Alpha() -> (Is in FrontEnd Alpha:double)
   ) -> (Pose) {
    alphaBoolBlend {blendOption=Linear}
    controlRigAssetReference {blueprintRigClass=&/M_LRG_OysterKnock_Cotton_CtrlRig.M_LRG_OysterKnock_Cotton_CtrlRig_C}
   }
   n3 = Anim Dynamics - Bone: dyn_l_kneerope_jnt!(
    SimulationSpace:AnimPhysSimSpaceType Enum
    ComponentPose=Local To Component(LocalPose=@n2)
    Alpha=1
    GravityOverride:Vector="0,0,0"
    ExternalForce="0,0,0"
   ) -> (Pose) {
    angularDampingOverride=1
    angularSpringConstant=200
    bAngularSpring=true
    bOverrideAngularDamping=true
    bUseGravityOverride=true
    simulationSpace=RootRelative
    alphaBoolBlend {blendOption=Linear}
    boundBone {boneName=dyn_l_kneerope_jnt}
    componentAppliedLinearAccClamp=<6000 6000 6000>
    componentLinearAccScale=<0.3 0.3 0.3>
    componentLinearVelScale=<0.3 0.3 0.3>
    physicsBodyDefinitions [{
      boundBone {boneName=dyn_l_kneerope_jnt}
      boxExtents=<15 15 15>
      constraintSetup {angularLimitsMax=<1 1 5>, angularLimitsMin=<-1 -5 -5>, angularTarget=<1 0 0>}
      localJointOffset=<5 0 0>
     }]
   }
   n4 = Anim Dynamics - Bone: dyn_l_innerrope_01_jnt!(SimulationSpace:AnimPhysSimSpaceType Enum, ComponentPose=@n3, Alpha=1, GravityOverride:Vector="0,0,0", ExternalForce="0,0,0") -> (Pose) {
    angularDampingOverride=1
    angularSpringConstant=200
    bAngularSpring=true
    bOverrideAngularDamping=true
    bUseGravityOverride=true
    simulationSpace=RootRelative
    alphaBoolBlend {blendOption=Linear}
    boundBone {boneName=dyn_l_innerrope_01_jnt}
    componentAppliedLinearAccClamp=<6000 6000 6000>
    componentLinearAccScale=<0.3 0.3 0.3>
    componentLinearVelScale=<0.3 0.3 0.3>
    physicsBodyDefinitions [{
      boundBone {boneName=dyn_l_innerrope_01_jnt}
      boxExtents=<15 15 15>
      constraintSetup {angularLimitsMax=<1 1 5>, angularLimitsMin=<-1 -5 -5>, angularTarget=<1 0 0>}
      localJointOffset=<5 0 0>
     }]
   }
   n5 = Anim Dynamics - Bone: dyn_l_innerrope_02_jnt!(SimulationSpace:AnimPhysSimSpaceType Enum, ComponentPose=@n4, Alpha=1, GravityOverride:Vector="0,0,0", ExternalForce="0,0,0") -> (Pose) {
    angularDampingOverride=1
    angularSpringConstant=200
    bAngularSpring=true
    bOverrideAngularDamping=true
    bUseGravityOverride=true
    simulationSpace=RootRelative
    alphaBoolBlend {blendOption=Linear}
    boundBone {boneName=dyn_l_innerrope_02_jnt}
    componentAppliedLinearAccClamp=<6000 6000 6000>
    componentLinearAccScale=<0.3 0.3 0.3>
    componentLinearVelScale=<0.3 0.3 0.3>
    physicsBodyDefinitions [{
      boundBone {boneName=dyn_l_innerrope_02_jnt}
      boxExtents=<15 15 15>
      constraintSetup {angularLimitsMax=<1 1 5>, angularLimitsMin=<-1 -5 -5>, angularTarget=<1 0 0>}
      localJointOffset=<5 0 0>
     }]
   }
   n6 = Anim Dynamics - Bone: dyn_l_innerrope_03_jnt!(SimulationSpace:AnimPhysSimSpaceType Enum, ComponentPose=@n5, Alpha=1, GravityOverride:Vector="0,0,0", ExternalForce="0,0,0") -> (Pose) {
    angularDampingOverride=1
    angularSpringConstant=200
    bAngularSpring=true
    bOverrideAngularDamping=true
    bUseGravityOverride=true
    simulationSpace=RootRelative
    alphaBoolBlend {blendOption=Linear}
    boundBone {boneName=dyn_l_innerrope_03_jnt}
    componentAppliedLinearAccClamp=<6000 6000 6000>
    componentLinearAccScale=<0.3 0.3 0.3>
    componentLinearVelScale=<0.3 0.3 0.3>
    physicsBodyDefinitions [{
      boundBone {boneName=dyn_l_innerrope_03_jnt}
      boxExtents=<15 15 15>
      constraintSetup {angularLimitsMax=<1 5 5>, angularLimitsMin=<-1 -5 -5>, angularTarget=<1 0 0>}
      localJointOffset=<5 0 0>
     }]
   }
   n7 = Anim Dynamics - Bone: dyn_l_outterrope_01_jnt!(SimulationSpace:AnimPhysSimSpaceType Enum, ComponentPose=@n6, Alpha=1, GravityOverride:Vector="0,0,0", ExternalForce="0,0,0") -> (Pose) {
    angularDampingOverride=1
    angularSpringConstant=200
    bAngularSpring=true
    bOverrideAngularDamping=true
    bUseGravityOverride=true
    simulationSpace=RootRelative
    alphaBoolBlend {blendOption=Linear}
    boundBone {boneName=dyn_l_outterrope_01_jnt}
    componentAppliedLinearAccClamp=<6000 6000 6000>
    componentLinearAccScale=<0.3 0.3 0.3>
    componentLinearVelScale=<0.3 0.3 0.3>
    physicsBodyDefinitions [{
      boundBone {boneName=dyn_l_outterrope_01_jnt}
      boxExtents=<15 15 15>
      constraintSetup {angularLimitsMax=<1 1 5>, angularLimitsMin=<-1 -5 -5>, angularTarget=<1 0 0>}
      localJointOffset=<5 0 0>
     }]
   }
   n8 = Anim Dynamics - Bone: dyn_l_outterrope_02_jnt!(SimulationSpace:AnimPhysSimSpaceType Enum, ComponentPose=@n7, Alpha=1, GravityOverride:Vector="0,0,0", ExternalForce="0,0,0") -> (Pose) {
    angularDampingOverride=1
    angularSpringConstant=200
    bAngularSpring=true
    bOverrideAngularDamping=true
    bUseGravityOverride=true
    simulationSpace=RootRelative
    alphaBoolBlend {blendOption=Linear}
    boundBone {boneName=dyn_l_outterrope_02_jnt}
    componentAppliedLinearAccClamp=<6000 6000 6000>
    componentLinearAccScale=<0.3 0.3 0.3>
    componentLinearVelScale=<0.3 0.3 0.3>
    physicsBodyDefinitions [{
      boundBone {boneName=dyn_l_outterrope_02_jnt}
      boxExtents=<15 15 15>
      constraintSetup {angularLimitsMax=<1 1 5>, angularLimitsMin=<-1 -5 -5>, angularTarget=<1 0 0>}
      localJointOffset=<5 0 0>
     }]
   }
   n9 = Anim Dynamics - Bone: dyn_l_outterrope_03_jnt!(SimulationSpace:AnimPhysSimSpaceType Enum, ComponentPose=@n8, Alpha=1, GravityOverride:Vector="0,0,0", ExternalForce="0,0,0") -> (Pose) {
    angularDampingOverride=1
    angularSpringConstant=200
    bAngularSpring=true
    bOverrideAngularDamping=true
    bUseGravityOverride=true
    simulationSpace=RootRelative
    alphaBoolBlend {blendOption=Linear}
    boundBone {boneName=dyn_l_outterrope_03_jnt}
    componentAppliedLinearAccClamp=<6000 6000 6000>
    componentLinearAccScale=<0.3 0.3 0.3>
    componentLinearVelScale=<0.3 0.3 0.3>
    physicsBodyDefinitions [{
      boundBone {boneName=dyn_l_outterrope_03_jnt}
      boxExtents=<15 15 15>
      constraintSetup {angularLimitsMax=<1 5 5>, angularLimitsMin=<-1 -5 -5>, angularTarget=<1 0 0>}
      localJointOffset=<5 0 0>
     }]
   }
   n10 = Anim Dynamics - Bone: dyn_r_kneerope_jnt!(SimulationSpace:AnimPhysSimSpaceType Enum, ComponentPose=@n9, Alpha=1, GravityOverride:Vector="0,0,0", ExternalForce="0,0,0") -> (Pose) {
    angularDampingOverride=1
    angularSpringConstant=200
    bAngularSpring=true
    bOverrideAngularDamping=true
    bUseGravityOverride=true
    simulationSpace=RootRelative
    alphaBoolBlend {blendOption=Linear}
    boundBone {boneName=dyn_r_kneerope_jnt}
    componentAppliedLinearAccClamp=<6000 6000 6000>
    componentLinearAccScale=<0.3 0.3 0.3>
    componentLinearVelScale=<0.3 0.3 0.3>
    physicsBodyDefinitions [{
      boundBone {boneName=dyn_r_kneerope_jnt}
      boxExtents=<15 15 15>
      constraintSetup {angularLimitsMax=<1 1 5>, angularLimitsMin=<-1 -5 -5>, angularTarget=<1 0 0>}
      localJointOffset=<5 0 0>
     }]
   }
   n11 = Anim Dynamics - Bone: dyn_r_innerrope_01_jnt!(SimulationSpace:AnimPhysSimSpaceType Enum, ComponentPose=@n10, Alpha=1, GravityOverride:Vector="0,0,0", ExternalForce="0,0,0") -> (Pose) {
    angularDampingOverride=1
    angularSpringConstant=200
    bAngularSpring=true
    bOverrideAngularDamping=true
    bUseGravityOverride=true
    simulationSpace=RootRelative
    alphaBoolBlend {blendOption=Linear}
    boundBone {boneName=dyn_r_innerrope_01_jnt}
    componentAppliedLinearAccClamp=<6000 6000 6000>
    componentLinearAccScale=<0.3 0.3 0.3>
    componentLinearVelScale=<0.3 0.3 0.3>
    physicsBodyDefinitions [{
      boundBone {boneName=dyn_r_innerrope_01_jnt}
      boxExtents=<15 15 15>
      constraintSetup {angularLimitsMax=<1 1 5>, angularLimitsMin=<-1 -5 -5>, angularTarget=<1 0 0>}
      localJointOffset=<5 0 0>
     }]
   }
   n12 = Anim Dynamics - Bone: dyn_r_innerrope_02_jnt!(SimulationSpace:AnimPhysSimSpaceType Enum, ComponentPose=@n11, Alpha=1, GravityOverride:Vector="0,0,0", ExternalForce="0,0,0") -> (Pose) {
    angularDampingOverride=1
    angularSpringConstant=200
    bAngularSpring=true
    bOverrideAngularDamping=true
    bUseGravityOverride=true
    simulationSpace=RootRelative
    alphaBoolBlend {blendOption=Linear}
    boundBone {boneName=dyn_r_innerrope_02_jnt}
    componentAppliedLinearAccClamp=<6000 6000 6000>
    componentLinearAccScale=<0.3 0.3 0.3>
    componentLinearVelScale=<0.3 0.3 0.3>
    physicsBodyDefinitions [{
      boundBone {boneName=dyn_r_innerrope_02_jnt}
      boxExtents=<15 15 15>
      constraintSetup {angularLimitsMax=<1 1 5>, angularLimitsMin=<-1 -5 -5>, angularTarget=<1 0 0>}
      localJointOffset=<5 0 0>
     }]
   }
   n13 = Anim Dynamics - Bone: dyn_r_innerrope_03_jnt!(SimulationSpace:AnimPhysSimSpaceType Enum, ComponentPose=@n12, Alpha=1, GravityOverride:Vector="0,0,0", ExternalForce="0,0,0") -> (Pose) {
    angularDampingOverride=1
    angularSpringConstant=200
    bAngularSpring=true
    bOverrideAngularDamping=true
    bUseGravityOverride=true
    simulationSpace=RootRelative
    alphaBoolBlend {blendOption=Linear}
    boundBone {boneName=dyn_r_innerrope_03_jnt}
    componentAppliedLinearAccClamp=<6000 6000 6000>
    componentLinearAccScale=<0.3 0.3 0.3>
    componentLinearVelScale=<0.3 0.3 0.3>
    physicsBodyDefinitions [{
      boundBone {boneName=dyn_r_innerrope_03_jnt}
      boxExtents=<15 15 15>
      constraintSetup {angularLimitsMax=<1 5 5>, angularLimitsMin=<-1 -5 -5>, angularTarget=<1 0 0>}
      localJointOffset=<5 0 0>
     }]
   }
   n14 = Anim Dynamics - Bone: dyn_r_outterrope_01_jnt!(SimulationSpace:AnimPhysSimSpaceType Enum, ComponentPose=@n13, Alpha=1, GravityOverride:Vector="0,0,0", ExternalForce="0,0,0") -> (Pose) {
    angularDampingOverride=1
    angularSpringConstant=200
    bAngularSpring=true
    bOverrideAngularDamping=true
    bUseGravityOverride=true
    simulationSpace=RootRelative
    alphaBoolBlend {blendOption=Linear}
    boundBone {boneName=dyn_r_outterrope_01_jnt}
    componentAppliedLinearAccClamp=<6000 6000 6000>
    componentLinearAccScale=<0.3 0.3 0.3>
    componentLinearVelScale=<0.3 0.3 0.3>
    physicsBodyDefinitions [{
      boundBone {boneName=dyn_r_outterrope_01_jnt}
      boxExtents=<15 15 15>
      constraintSetup {angularLimitsMax=<1 1 5>, angularLimitsMin=<-1 -5 -5>, angularTarget=<1 0 0>}
      localJointOffset=<5 0 0>
     }]
   }
   n15 = Anim Dynamics - Bone: dyn_r_outterrope_02_jnt!(SimulationSpace:AnimPhysSimSpaceType Enum, ComponentPose=@n14, Alpha=1, GravityOverride:Vector="0,0,0", ExternalForce="0,0,0") -> (Pose) {
    angularDampingOverride=1
    angularSpringConstant=200
    bAngularSpring=true
    bOverrideAngularDamping=true
    bUseGravityOverride=true
    simulationSpace=RootRelative
    alphaBoolBlend {blendOption=Linear}
    boundBone {boneName=dyn_r_outterrope_02_jnt}
    componentAppliedLinearAccClamp=<6000 6000 6000>
    componentLinearAccScale=<0.3 0.3 0.3>
    componentLinearVelScale=<0.3 0.3 0.3>
    physicsBodyDefinitions [{
      boundBone {boneName=dyn_r_outterrope_02_jnt}
      boxExtents=<15 15 15>
      constraintSetup {angularLimitsMax=<1 1 5>, angularLimitsMin=<-1 -5 -5>, angularTarget=<1 0 0>}
      localJointOffset=<5 0 0>
     }]
   }
   n16 = Anim Dynamics - Bone: dyn_r_outterrope_03_jnt!(SimulationSpace:AnimPhysSimSpaceType Enum, ComponentPose=@n15, Alpha=1, GravityOverride:Vector="0,0,0", ExternalForce="0,0,0") -> (Pose) {
    angularDampingOverride=1
    angularSpringConstant=200
    bAngularSpring=true
    bOverrideAngularDamping=true
    bUseGravityOverride=true
    simulationSpace=RootRelative
    alphaBoolBlend {blendOption=Linear}
    boundBone {boneName=dyn_r_outterrope_03_jnt}
    componentAppliedLinearAccClamp=<6000 6000 6000>
    componentLinearAccScale=<0.3 0.3 0.3>
    componentLinearVelScale=<0.3 0.3 0.3>
    physicsBodyDefinitions [{
      boundBone {boneName=dyn_r_outterrope_03_jnt}
      boxExtents=<15 15 15>
      constraintSetup {angularLimitsMax=<1 5 5>, angularLimitsMin=<-1 -5 -5>, angularTarget=<1 0 0>}
      localJointOffset=<5 0 0>
     }]
   }
   n17 = Anim Dynamics - Bone: dyn_ropebelt_jnt!(SimulationSpace:AnimPhysSimSpaceType Enum, ComponentPose=@n16, Alpha=1, GravityOverride:Vector="0,0,0", ExternalForce="0,0,0") -> (Pose) {
    angularDampingOverride=1
    angularSpringConstant=200
    bAngularSpring=true
    bOverrideAngularDamping=true
    bUseGravityOverride=true
    simulationSpace=RootRelative
    alphaBoolBlend {blendOption=Linear}
    boundBone {boneName=dyn_ropebelt_jnt}
    componentAppliedLinearAccClamp=<6000 6000 6000>
    componentLinearAccScale=<0.3 0.3 0.3>
    componentLinearVelScale=<0.3 0.3 0.3>
    physicsBodyDefinitions [{
      boundBone {boneName=dyn_ropebelt_jnt}
      boxExtents=<30 30 30>
      constraintSetup {angularLimitsMax=<1 5 5>, angularLimitsMin=<-1 -5 -5>, angularTarget=<1 0 0>}
      localJointOffset=<10 0 0>
     }]
   }
   n18 = Anim Dynamics - Bone: dyn_l_ropebelt_01_jnt!(SimulationSpace:AnimPhysSimSpaceType Enum, ComponentPose=@n17, Alpha=1, GravityOverride:Vector="0,0,0", ExternalForce="0,0,0") -> (Pose) {
    angularDampingOverride=1
    angularSpringConstant=200
    bAngularSpring=true
    bOverrideAngularDamping=true
    bUseGravityOverride=true
    simulationSpace=RootRelative
    alphaBoolBlend {blendOption=Linear}
    boundBone {boneName=dyn_l_ropebelt_01_jnt}
    componentAppliedLinearAccClamp=<6000 6000 6000>
    componentLinearAccScale=<0.3 0.3 0.3>
    componentLinearVelScale=<0.3 0.3 0.3>
    physicsBodyDefinitions [{
      boundBone {boneName=dyn_l_ropebelt_01_jnt}
      boxExtents=<30 30 30>
      constraintSetup {angularLimitsMax=<1 0 15>, angularLimitsMin=<-1 -5 -15>, angularTarget=<1 0 0>}
      localJointOffset=<10 0 0>
     }]
   }
   n19 = Anim Dynamics - Bone: dyn_l_ropebelt_02_jnt!(SimulationSpace:AnimPhysSimSpaceType Enum, ComponentPose=@n18, Alpha=1, GravityOverride:Vector="0,0,0", ExternalForce="0,0,0") -> (Pose) {
    angularDampingOverride=1
    angularSpringConstant=200
    bAngularSpring=true
    bOverrideAngularDamping=true
    bUseGravityOverride=true
    simulationSpace=RootRelative
    alphaBoolBlend {blendOption=Linear}
    boundBone {boneName=dyn_l_ropebelt_02_jnt}
    componentAppliedLinearAccClamp=<6000 6000 6000>
    componentLinearAccScale=<0.3 0.3 0.3>
    componentLinearVelScale=<0.3 0.3 0.3>
    physicsBodyDefinitions [{
      boundBone {boneName=dyn_l_ropebelt_02_jnt}
      boxExtents=<30 30 30>
      constraintSetup {angularLimitsMax=<1 0 15>, angularLimitsMin=<-1 -5 -15>, angularTarget=<1 0 0>}
      localJointOffset=<10 0 0>
     }]
   }
   n20 = Anim Dynamics - Bone: dyn_l_ropebelt_03_jnt!(SimulationSpace:AnimPhysSimSpaceType Enum, ComponentPose=@n19, Alpha=1, GravityOverride:Vector="0,0,0", ExternalForce="0,0,0") -> (Pose) {
    angularDampingOverride=1
    angularSpringConstant=200
    bAngularSpring=true
    bOverrideAngularDamping=true
    bUseGravityOverride=true
    simulationSpace=RootRelative
    alphaBoolBlend {blendOption=Linear}
    boundBone {boneName=dyn_l_ropebelt_03_jnt}
    componentAppliedLinearAccClamp=<6000 6000 6000>
    componentLinearAccScale=<0.3 0.3 0.3>
    componentLinearVelScale=<0.3 0.3 0.3>
    physicsBodyDefinitions [{
      boundBone {boneName=dyn_l_ropebelt_03_jnt}
      boxExtents=<30 30 30>
      constraintSetup {angularLimitsMax=<1 0 15>, angularLimitsMin=<-1 -5 -15>, angularTarget=<1 0 0>}
      localJointOffset=<10 0 0>
     }]
   }
   n21 = Anim Dynamics - Bone: dyn_r_ropebelt_01_jnt!(SimulationSpace:AnimPhysSimSpaceType Enum, ComponentPose=@n20, Alpha=1, GravityOverride:Vector="0,0,0", ExternalForce="0,0,0") -> (Pose) {
    angularDampingOverride=1
    angularSpringConstant=200
    bAngularSpring=true
    bOverrideAngularDamping=true
    bUseGravityOverride=true
    simulationSpace=RootRelative
    alphaBoolBlend {blendOption=Linear}
    boundBone {boneName=dyn_r_ropebelt_01_jnt}
    componentAppliedLinearAccClamp=<6000 6000 6000>
    componentLinearAccScale=<0.3 0.3 0.3>
    componentLinearVelScale=<0.3 0.3 0.3>
    physicsBodyDefinitions [{
      boundBone {boneName=dyn_r_ropebelt_01_jnt}
      boxExtents=<30 30 30>
      constraintSetup {angularLimitsMax=<1 0 15>, angularLimitsMin=<-1 -5 -15>, angularTarget=<1 0 0>}
      localJointOffset=<10 0 0>
     }]
   }
   n22 = Anim Dynamics - Bone: dyn_r_ropebelt_02_jnt!(SimulationSpace:AnimPhysSimSpaceType Enum, ComponentPose=@n21, Alpha=1, GravityOverride:Vector="0,0,0", ExternalForce="0,0,0") -> (Pose) {
    angularDampingOverride=1
    angularSpringConstant=200
    bAngularSpring=true
    bOverrideAngularDamping=true
    bUseGravityOverride=true
    simulationSpace=RootRelative
    alphaBoolBlend {blendOption=Linear}
    boundBone {boneName=dyn_r_ropebelt_02_jnt}
    componentAppliedLinearAccClamp=<6000 6000 6000>
    componentLinearAccScale=<0.3 0.3 0.3>
    componentLinearVelScale=<0.3 0.3 0.3>
    physicsBodyDefinitions [{
      boundBone {boneName=dyn_r_ropebelt_02_jnt}
      boxExtents=<30 30 30>
      constraintSetup {angularLimitsMax=<1 0 15>, angularLimitsMin=<-1 -5 -15>, angularTarget=<1 0 0>}
      localJointOffset=<10 0 0>
     }]
   }
   n23 = Anim Dynamics - Bone: dyn_r_ropebelt_03_jnt!(SimulationSpace:AnimPhysSimSpaceType Enum, ComponentPose=@n22, Alpha=1, GravityOverride:Vector="0,0,0", ExternalForce="0,0,0") -> (Pose) {
    angularDampingOverride=1
    angularSpringConstant=200
    bAngularSpring=true
    bOverrideAngularDamping=true
    bUseGravityOverride=true
    simulationSpace=RootRelative
    alphaBoolBlend {blendOption=Linear}
    boundBone {boneName=dyn_r_ropebelt_03_jnt}
    componentAppliedLinearAccClamp=<6000 6000 6000>
    componentLinearAccScale=<0.3 0.3 0.3>
    componentLinearVelScale=<0.3 0.3 0.3>
    physicsBodyDefinitions [{
      boundBone {boneName=dyn_r_ropebelt_03_jnt}
      boxExtents=<30 30 30>
      constraintSetup {angularLimitsMax=<1 0 15>, angularLimitsMin=<-1 -5 -15>, angularTarget=<1 0 0>}
      localJointOffset=<10 0 0>
     }]
   }
   Output Pose(Result=Component To Local(ComponentPose=@n23))
  ]}
 EventGraph {nodes [
   Event BlueprintUpdateAnimation()
   > n1 = Sequence!(execute) -> (then_0, then_1, then_2:Exec, then_3:Exec, then_4:Exec, then_5:Exec)
   > n2 = Gravity Override State Function DEPRECATED(CharacterPartAnimInstance=Self-Reference(), GameTimeInSeconds=GetGameTimeInSeconds(), GravityOverrideParams=Get Gravity Override Params())
   > Set GravityOverride(GravityOverride=@n2)
   Set simSpace(execute=@n1.then_1, simSpace:AnimPhysSimSpaceType Enum) -> (Output_Get:AnimPhysSimSpaceType Enum)
   > Branch(Condition=OR Boolean(A=Get bIsDBNO(), B=Get bIsInFrontEnd()))
   > n3 = Branch(Condition=OR Boolean(A=Get bIsDBNOCarried(), B=Get bIsInFrontEnd()))
   > Set simSpace(simSpace=World)
   Set simSpace(execute=@n3.else)
   n4 = Get bIsPlayingEmote()
   Set Is not Playing Emote Alpha(execute=@n1.then_2, Is not Playing Emote Alpha:double="To Float (Boolean)"(InBool=NOT Boolean(A=@n4))) -> (Output_Get:double)
   n5 = Get bIsSkydiveFloating()
   n6 = Get bIsSkydiveDiveMode()
   > Set Is Skydiving Alpha(Is Skydiving Alpha:double="To Float (Boolean)"(InBool=OR Boolean(A=@n5, B=@n6))) -> (Output_Get:double)
   > Set Is Parachute Opened(Is Parachute Opened:double="To Float (Boolean)"(InBool=Get bIsParachuteOpened())) -> (Output_Get:double)
   > Set Is in FrontEnd Alpha(Is in FrontEnd Alpha:double="To Float (Boolean)"(InBool=Get bIsInFrontEnd())) -> (Output_Get:double)
   n7 = Sequence!(execute=@n1.then_3) -> (then_0, then_1, then_2, then_3, then_4)
   > n8 = Cast To FortPlayerAnimInstance(Object=GetAnimInstance(self=GetOwnerSkeletalMeshComponent()))
   > Set isPistol(isPistol=false)
   > Set isRifle(isRifle:bool=false) -> (Output_Get:bool)
   > n9 = Switch on EFortWeaponCoreAnimation!(execute, Selection:EFortWeaponCoreAnimation Enum=Get WeaponCoreAnim(self=@n8)) -> (
    Melee:Exec
    Pistol:Exec
    Shotgun:Exec
    PaperBlueprint:Exec
    Rifle:Exec
    MeleeOneHand:Exec
    MachinePistol:Exec
    RocketLauncher:Exec
    GrenadeLauncher:Exec
    GoingCommando:Exec
    AssaultRifle:Exec
    TacticalShotgun:Exec
    SniperRifle:Exec
    TrapPlacement:Exec
    ShoulderLauncher:Exec
    AbilityDecoTool:Exec
    Crossbow:Exec
    C4:Exec
    RemoteControl:Exec
    DualWield:Exec
    AR_BullPup:Exec
    AR_ForwardGrip:Exec
    MedPackPaddles:Exec
    SMG_P90:Exec
    AR_DrumGun:Exec
    Consumable_Small:Exec
    Consumable_Large:Exec
    Balloon:Exec
    MountedTurret:Exec
    CreativeTool:Exec
    ExplosiveBow:Exec
    AshtonIndigo:Exec
    AshtonChicago:Exec
    MeleeDualWield:Exec
    Unarmed:Exec
   )
   Set isRifle(execute=[@n9.AssaultRifle, @n9.Rifle, @n9.RocketLauncher, @n9.Shotgun, @n9.SniperRifle, @n9.TacticalShotgun], isRifle=true)
   Set isPistol(execute=@n9.Pistol, isPistol=true)
   Set Is Rifle Alpha(execute=@n7.then_1, Is Rifle Alpha:double="To Float (Boolean)"(InBool=Get isRifle() -> (isRifle:bool))) -> (Output_Get:double)
   Set Is Pistol Alpha(execute=@n7.then_2, Is Pistol Alpha:double="To Float (Boolean)"(InBool=Get isPistol() -> (isPistol:bool))) -> (Output_Get:double)
   n10 = GetOwnerSkeletalMeshComponent()
   n11 = BreakRotator(InRot=BreakTransform(InTransform=GetDeltaTransformFromRefPose(self=@n10, BoneName=head, BaseName=neck_02)).Rotation)
   n12 = Get Z Percentage() -> (Z Percentage:double)
   n13 = Get X Percentage() -> (X Percentage:double)
   n14 = Get Y Percentage() -> (Y Percentage:double)
   Set Head Rot(
    execute=@n1.then_4
    Head Rot:Rotator=MakeRotator(
     Roll=float * float!(A=@n11, B=-1, C:double=@n13) -> (ReturnValue) {is=title K2Node_PromotableOperator}
     Pitch=float * float(A=@n11.Pitch, B=-1, C=@n14)
     Yaw=float * float(A=@n11.Yaw, B=-1, C=@n12)
    )
   ) -> (Output_Get:Rotator)
   n15 = Branch(execute=@n1.then_5, Condition=Get bIsInFrontEnd())
   > Set Animdynamics Alpha(Animdynamics Alpha:double=0.25) -> (Output_Get:double)
   Set Animdynamics Alpha(execute=@n15.else, Animdynamics Alpha=1)
   TryGetPawnOwner()
  ]}
]
variables [
 Animdynamics Alpha {type=number}
 Gravity Override Params {type=GravityOverrideParamsStruct}
 GravityOverride {type=Vector}
 Head Rot {type=Rotator}
 Is in FrontEnd Alpha {type=number}
 Is not Playing Emote Alpha {type=number}
 Is Parachute Opened {type=number}
 Is Pistol Alpha {type=number}
 Is Rifle Alpha {type=number}
 Is Skydiving Alpha {type=number}
 isPistol {type=boolean}
 isRifle {type=boolean}
 simSpace {type=AnimPhysSimSpaceType}
 X Percentage {type=number}
 Y Percentage {type=number}
 Z Percentage {type=number, default=0.75}
]
