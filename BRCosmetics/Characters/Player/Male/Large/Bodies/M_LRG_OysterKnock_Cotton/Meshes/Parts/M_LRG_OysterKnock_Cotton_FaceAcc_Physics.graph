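# Physics asset for the face accessory of the large male OysterKnock Cotton player cosmetic, covering a chin rope and ribbons.
# A central chin-rope joint, linked to the head through a master bone, branches into left and right four-joint rope chains and three-joint ribbon chains with their own masters.
# Chain links have tight swing and near-locked twist limits, while masters and the head-neck-upper-spine capsule chain rotate freely.
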PhysicsAsset /BRCosmetics/Characters/Player/Male/Large/Bodies/M_LRG_OysterKnock_Cotton/Meshes/Parts/M_LRG_OysterKnock_Cotton_FaceAcc_Physics
bodies {
 dyn_chinrope_MSTR [{shapeType=Capsule, length=1, radius=1, shapeName=dyn_chinrope_MSTR_capsule, rotation=<rot 0 0 90>}]
 dyn_chinrope_jnt [{shapeType=Capsule, length=1, radius=1, center=<1 0 -1>, rotation=<rot 0 0 90>}]
 dyn_l_chinrope_01_jnt [{shapeType=Capsule, length=2, radius=1, shapeName=dyn_l_chinrope_01_jnt_capsule, center=<2 0 0>, rotation=<rot 0 90 90>}]
 dyn_l_chinrope_02_jnt [{shapeType=Capsule, length=2, radius=1, shapeName=dyn_l_chinrope_02_jnt_capsule, center=<2 0 0>, rotation=<rot 0 90 90>}]
 dyn_l_chinrope_03_jnt [{shapeType=Capsule, length=2, radius=1, shapeName=dyn_l_chinrope_03_jnt_capsule, center=<2 0 0>, rotation=<rot 0 90 90>}]
 dyn_l_chinrope_04_jnt [{shapeType=Capsule, length=2, radius=1, shapeName=dyn_l_chinrope_04_jnt_capsule, center=<2 0 0>, rotation=<rot 0 90 90>}]
 dyn_l_ribbon_01_jnt [{shapeType=Capsule, length=3, radius=2.5, shapeName=dyn_l_ribbon_01_jnt_capsule, center=<2 0 0.5>, rotation=<rot 0 0 90>}]
 dyn_l_ribbon_02_jnt [{shapeType=Capsule, length=3, radius=2.5, shapeName=dyn_l_ribbon_02_jnt_capsule, center=<2 0 0.5>, rotation=<rot 0 0 90>}]
 dyn_l_ribbon_03_jnt [{shapeType=Capsule, length=3, radius=2.5, shapeName=dyn_l_ribbon_03_jnt_capsule, center=<2 0 0.5>, rotation=<rot 0 0 90>}]
 dyn_l_ribbon_MSTR [{shapeType=Capsule, length=1, radius=1, shapeName=dyn_l_ribbon_MSTR_capsule, rotation=<rot 0 90 90>}]
 dyn_r_chinrope_01_jnt [{shapeType=Capsule, length=2, radius=1, shapeName=dyn_r_chinrope_01_jnt_capsule, center=<2 0 0>, rotation=<rot 0 90 90>}]
 dyn_r_chinrope_02_jnt [{shapeType=Capsule, length=2, radius=1, shapeName=dyn_r_chinrope_02_jnt_capsule, center=<2 0 0>, rotation=<rot 0 90 90>}]
 dyn_r_chinrope_03_jnt [{shapeType=Capsule, length=2, radius=1, shapeName=dyn_r_chinrope_03_jnt_capsule, center=<2 0 0>, rotation=<rot 0 90 90>}]
 dyn_r_chinrope_04_jnt [{shapeType=Capsule, length=2, radius=1, shapeName=dyn_r_chinrope_04_jnt_capsule, center=<2 0 0>, rotation=<rot 0 90 90>}]
 dyn_r_ribbon_01_jnt [{shapeType=Capsule, length=3, radius=2, shapeName=dyn_r_ribbon_01_jnt_capsule, center=<2 0 1>, rotation=<rot 0 0 90>}]
 dyn_r_ribbon_02_jnt [{shapeType=Capsule, length=4, radius=2, shapeName=dyn_r_ribbon_02_jnt_capsule, center=<2 0 1>, rotation=<rot 0 0 90>}]
 dyn_r_ribbon_03_jnt [{shapeType=Capsule, length=3, radius=2, shapeName=dyn_r_ribbon_03_jnt_capsule, center=<2 0 1>, rotation=<rot 0 0 90>}]
 dyn_r_ribbon_MSTR [{shapeType=Capsule, length=1, radius=1, shapeName=dyn_r_ribbon_MSTR_capsule, rotation=<rot 0 90 90>}]
 head [{shapeType=Capsule, length=4, radius=10, center=<6 -1.5 0>}]
 neck_01 [{shapeType=Capsule, length=16, radius=16, shapeName=neck_01_capsule, center=<-4 0 0>}]
 neck_02 [{shapeType=Capsule, length=8, radius=10, shapeName=neck_02_capsule, center=<-4 0 0>}]
 spine_04 [{shapeType=Capsule, length=16, radius=18, shapeName=spine_04_capsule}]
 spine_05 [{shapeType=Capsule, length=16, radius=24, shapeName=spine_05_capsule, center=<-4 0 0>}]
}
constraints [
 {bone1Name=dyn_chinrope_jnt, bone2Name=dyn_chinrope_MSTR, swing1LimitDegrees=45, swing2LimitDegrees=45, twistLimitDegrees=45}
 {
  bone1Name=head
  bone2Name=neck_02
  swing1LimitDegrees=45
  swing1Motion=Free
  swing2LimitDegrees=45
  swing2Motion=Free
  twistLimitDegrees=45
  twistMotion=Free
 }
 {
  bone1Name=neck_01
  bone2Name=spine_05
  swing1LimitDegrees=45
  swing1Motion=Free
  swing2LimitDegrees=45
  swing2Motion=Free
  twistLimitDegrees=45
  twistMotion=Free
 }
 {
  bone1Name=neck_02
  bone2Name=neck_01
  swing1LimitDegrees=45
  swing1Motion=Free
  swing2LimitDegrees=45
  swing2Motion=Free
  twistLimitDegrees=45
  twistMotion=Free
 }
 {
  bone1Name=spine_05
  bone2Name=spine_04
  swing1LimitDegrees=45
  swing1Motion=Free
  swing2LimitDegrees=45
  swing2Motion=Free
  twistLimitDegrees=45
  twistMotion=Free
 }
 {
  bone1Name=dyn_chinrope_MSTR
  bone2Name=head
  swing1LimitDegrees=45
  swing1Motion=Free
  swing2LimitDegrees=45
  swing2Motion=Free
  twistLimitDegrees=45
  twistMotion=Free
 }
 {bone1Name=dyn_l_chinrope_01_jnt, bone2Name=dyn_chinrope_jnt, swing1LimitDegrees=15, swing2LimitDegrees=15, twistLimitDegrees=1}
 {bone1Name=dyn_l_chinrope_02_jnt, bone2Name=dyn_l_chinrope_01_jnt, swing1LimitDegrees=15, swing2LimitDegrees=15, twistLimitDegrees=1}
 {bone1Name=dyn_l_chinrope_03_jnt, bone2Name=dyn_l_chinrope_02_jnt, swing1LimitDegrees=15, swing2LimitDegrees=15, twistLimitDegrees=1}
 {bone1Name=dyn_l_chinrope_04_jnt, bone2Name=dyn_l_chinrope_03_jnt, swing1LimitDegrees=15, swing2LimitDegrees=15, twistLimitDegrees=1}
 {
  bone1Name=dyn_l_ribbon_MSTR
  bone2Name=dyn_chinrope_jnt
  swing1LimitDegrees=45
  swing1Motion=Free
  swing2LimitDegrees=45
  swing2Motion=Free
  twistLimitDegrees=45
  twistMotion=Free
 }
 {bone1Name=dyn_l_ribbon_01_jnt, bone2Name=dyn_l_ribbon_MSTR, swing1LimitDegrees=15, swing2LimitDegrees=15, twistLimitDegrees=2}
 {bone1Name=dyn_l_ribbon_02_jnt, bone2Name=dyn_l_ribbon_01_jnt, swing1LimitDegrees=15, swing2LimitDegrees=15, twistLimitDegrees=2}
 {bone1Name=dyn_l_ribbon_03_jnt, bone2Name=dyn_l_ribbon_02_jnt, swing1LimitDegrees=15, swing2LimitDegrees=15, twistLimitDegrees=2}
 {bone1Name=dyn_r_chinrope_01_jnt, bone2Name=dyn_chinrope_jnt, swing1LimitDegrees=15, swing2LimitDegrees=15, twistLimitDegrees=1}
 {bone1Name=dyn_r_chinrope_02_jnt, bone2Name=dyn_r_chinrope_01_jnt, swing1LimitDegrees=15, swing2LimitDegrees=15, twistLimitDegrees=1}
 {bone1Name=dyn_r_chinrope_03_jnt, bone2Name=dyn_r_chinrope_02_jnt, swing1LimitDegrees=15, swing2LimitDegrees=15, twistLimitDegrees=1}
 {bone1Name=dyn_r_chinrope_04_jnt, bone2Name=dyn_r_chinrope_03_jnt, swing1LimitDegrees=15, swing2LimitDegrees=15, twistLimitDegrees=1}
 {
  bone1Name=dyn_r_ribbon_MSTR
  bone2Name=dyn_chinrope_jnt
  swing1LimitDegrees=45
  swing1Motion=Free
  swing2LimitDegrees=45
  swing2Motion=Free
  twistLimitDegrees=45
  twistMotion=Free
 }
 {bone1Name=dyn_r_ribbon_01_jnt, bone2Name=dyn_r_ribbon_MSTR, swing1LimitDegrees=13, swing2LimitDegrees=13, twistLimitDegrees=2}
 {bone1Name=dyn_r_ribbon_02_jnt, bone2Name=dyn_r_ribbon_01_jnt, swing1LimitDegrees=13, swing2LimitDegrees=13, twistLimitDegrees=2}
 {bone1Name=dyn_r_ribbon_03_jnt, bone2Name=dyn_r_ribbon_02_jnt, swing1LimitDegrees=13, swing2LimitDegrees=13, twistLimitDegrees=2}
]
